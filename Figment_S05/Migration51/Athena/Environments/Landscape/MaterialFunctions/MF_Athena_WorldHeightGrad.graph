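# Landscape material function that outputs a height gradient from world-space Z.
# It subtracts an offset from the Z of the input coordinates and divides by a scale.
# Coordinates default to world position, with offset and scale defaulting to MPC_Landscape collection values, all overridable as inputs.
# The Alpha multiply and Min/Max clamp branch does not feed the output, so the result is unclamped; local scale and offset parameters are also unconnected.
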MaterialFunction /Figment_S05/Migration51/Athena/Environments/Landscape/MaterialFunctions/MF_Athena_WorldHeightGrad
graph {nodes [
  n1 = FunctionInput(Preview=WorldPosition().XYZ) {bUsePreviewValueAsDefault=true, inputName=Coords}
  n2 = CollectionParameter() {collection=/FigmentEnvironment/Migration51/Athena/Environments/Landscape/MPC/MPC_Landscape, parameterName=HeightGrad_Offset}
  n3 = Divide(
   A=Subtract(
    A=ComponentMask(None=@n1) {b=true}
    B=FunctionInput(Preview=@n2) {bUsePreviewValueAsDefault=true, inputName=height_Offset, inputType=FunctionInput_Scalar}
   )
   B=FunctionInput(Preview=CollectionParameter() {collection=&MPC_Landscape, parameterName=HeightGrad_Scale}) {bUsePreviewValueAsDefault=true, inputName=height_GradScale, inputType=FunctionInput_Scalar}
  )
  n4 = Constant()
  Clamp(
   None=Multiply(A=@n3, B=FunctionInput(Preview=@n4) {bUsePreviewValueAsDefault=true, inputName=Alpha, inputType=FunctionInput_Scalar})
   Min=ScalarParameter() {defaultValue=0.6, group=Height_Gradient, parameterName=Min_Value}
   Max=ScalarParameter() {defaultValue=1.05, group=Height_Gradient, parameterName=Max_Value}
  ) {maxDefault=1.25, minDefault=0.5}
  FunctionOutput(None=@n3)
  ScalarParameter() {defaultValue=15000, group=Height_Gradient, parameterName=HeightGrad_Scale}
  ScalarParameter() {defaultValue=-8000, group=Height_Gradient, parameterName=HeightGrad_Offset}
 ]}
