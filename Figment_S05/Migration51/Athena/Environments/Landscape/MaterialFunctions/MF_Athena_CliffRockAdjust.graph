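# Landscape material function that recolors cliff rock by biome, taking cliff material attributes plus farm and mountain alpha inputs.
# Base color blends toward the FarmGrass rock color by FarmAlpha, then the MountainGrass rock color by MountainAlpha.
# Roughness comes from the RoughRock parameter, while most other connected attributes, including Normal, read the input's Metallic output.
# Swamp blend, world-height multiply and biome mask nodes do not feed the output.
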MaterialFunction /Figment_S05/Migration51/Athena/Environments/Landscape/MaterialFunctions/MF_Athena_CliffRockAdjust
graph {nodes [
  Constant()
  n1 = BreakMaterialAttributes!(Attr=FunctionInput(Preview=MakeMaterialAttributes!(
     BaseColor
     Metallic
     Specular
     Roughness
     Anisotropy
     EmissiveColor
     Opacity
     OpacityMask
     Normal
     Tangent
     WorldPositionOffset
     SubsurfaceColor
     ClearCoat
     ClearCoatRoughness
     AmbientOcclusion
     Refraction
     CustomizedUVs_0
     CustomizedUVs_1
     CustomizedUVs_2
     CustomizedUVs_3
     CustomizedUVs_4
     CustomizedUVs_5
     CustomizedUVs_6
     CustomizedUVs_7
     PixelDepthOffset
     ShadingModel
     Displacement
    )) {inputName=CliffMaterial, inputType=FunctionInput_MaterialAttributes})
  n2 = LinearInterpolate(
   A=@n1.BaseColor
   B=MaterialFunctionCall!(Rock_Diffuse=@n1.BaseColor, Alpha, ColorBlend_Alpha, ColorBlend_Alpha_Bias, In, Grass_Diffuse) {materialFunction=&/MF_FarmGrass_Colors}.Rock
   Alpha=FunctionInput(Preview=Constant()) {bUsePreviewValueAsDefault=true, inputName=FarmAlpha, inputType=FunctionInput_Scalar}
  )
  n3 = MaterialFunctionCall!(Rock_Diffuse=@n2, ColorBlend_Alpha, ColorBlend_Alpha_Bias, Rock_Diffuse1, Grass_Diffuse) {materialFunction=&/MF_MountainGrass_Colors}
  n4 = FunctionInput(Preview=Constant()) {bUsePreviewValueAsDefault=true, inputName=MountainAlpha, inputType=FunctionInput_Scalar}
  FunctionOutput(None=MakeMaterialAttributes(
    BaseColor=LinearInterpolate(A=@n2, B=@n3.Rock, Alpha=@n4)
    Metallic=@n1.Metallic
    Roughness=ScalarParameter() {defaultValue=0.9, parameterName=RoughRock}
    EmissiveColor=@n1.Metallic
    Opacity=@n1.Metallic
    OpacityMask=@n1.Metallic
    Normal=@n1.Metallic
    WorldPositionOffset=@n1.Metallic
    SubsurfaceColor=@n1.Metallic
    ClearCoat=@n1.Metallic
    ClearCoatRoughness=@n1.Metallic
    AmbientOcclusion=@n1.Metallic
    Refraction=@n1.Metallic
    PixelDepthOffset=@n1.Metallic
   ))
  LinearInterpolate(
   A=@n2
   B=MaterialFunctionCall!(Rock_Diffuse=@n2, Grass_Diffuse) {materialFunction=&/MF_SwampGrass_Colors}.Rock
   Alpha=FunctionInput(Preview=Constant()) {bUsePreviewValueAsDefault=true, inputName=SwampAlpha, inputType=FunctionInput_Scalar}
  )
  MaterialFunctionCall() {materialFunction=/FigmentEnvironment/Migration51/Athena/Environments/Landscape/MaterialFunctions/MF_Athena_BiomeMask}
  Multiply(
   A=@n1.BaseColor
   B=MaterialFunctionCall!(Alpha, height_Offset, height_GradScale, Coords) {materialFunction=&/MF_Athena_WorldHeightGrad}.Result
  )
 ]}
