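# MF_StormMask is a landscape material function that outputs a mask of the area outside the Athena safe zone, where the storm is.
# It compares world-position XY to AthenaSafeZoneCenter XY with a SphereMask sized by AthenaSafeZoneRadius, both read from the FortniteMaterialParameters collection.
# Hardness of one gives a sharp edge, and a OneMinus inverts the result so pixels outside the circle are white.
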MaterialFunction /Figment_S05/Migration51/Athena/Environments/Landscape/MaterialFunctions/MF_StormMask
graph {nodes [
  n1 = ComponentMask(None=WorldPosition().XYZ) {g=true, r=true}
  n2 = CollectionParameter() {
   collection=/Figment_S05/Migration51/Packages/Fortress_SharedMaterials/GlobalMaterialParameters/FortniteMaterialParameters
   parameterName=AthenaSafeZoneCenter
  }
  n3 = ComponentMask(None=@n2) {g=true, r=true}
  n4 = CollectionParameter() {
   collection=&FortniteMaterialParameters
   parameterName=AthenaSafeZoneRadius
  }
  n5 = Constant() {r=1}
  FunctionOutput(None=OneMinus(None=SphereMask(A=@n1, B=@n3, Radius=@n4, Hardness=@n5)))
 ]}
